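# Physics asset for the Spy Tech Hacker backpack cosmetic, defining collision bodies and joints for its dangling straps.
# The pack bone carries two long, overlapping capsules, and each of the three strap bones has one small capsule.
# Straps 01 and 02 form a two-link chain from the pack, while strap 03 attaches to the pack separately.
# All three joints share the same asymmetric swing limits and a tight twist limit.
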PhysicsAsset /BRCosmetics/Accessories/FORT_Backpacks/Backpack_Spy_Tech_Hacker/Meshes/M_MED_Spy_Tech_Hacker_Pack_PhysicsAsset
bodies {
 dyn_strap_01 [{shapeType=Capsule, length=2, radius=1, center=<4 0 0>, rotation=<rot 0 0 90>}]
 dyn_strap_02 [{shapeType=Capsule, length=2, radius=1.3, center=<4 0 0>, rotation=<rot 0 0 90>}]
 dyn_strap_03 [{shapeType=Capsule, length=1, radius=1, center=<2 0 0.5>, rotation=<rot 90 0 0>}]
 pack [
  {shapeType=Capsule, length=21.7531, radius=4.7468, center=<2.7112 -4.3366 -0.8029>, rotation=<rot 88.9594 -180 -180>}
  {shapeType=Capsule, length=21.7531, radius=4.7469, center=<2.7527 -4.1351 -3.1202>, rotation=<rot 88.9476 -180 -180>}
 ]
}
constraints [
 {bone1Name=dyn_strap_01, bone2Name=pack, swing1LimitDegrees=20, swing2LimitDegrees=40, twistLimitDegrees=5}
 {bone1Name=dyn_strap_02, bone2Name=dyn_strap_01, swing1LimitDegrees=20, swing2LimitDegrees=40, twistLimitDegrees=5}
 {bone1Name=dyn_strap_03, bone2Name=pack, swing1LimitDegrees=20, swing2LimitDegrees=40, twistLimitDegrees=5}
]
